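# Material function for a translucent primitive material with a glitter layer on top.
# MF_Primitive_Trans_Base supplies the base attributes, with Refraction set on them, and passes its Thin output through as TransmittanceColor.
# MF_GlitterLayer supplies a normal and mask for a glitter attribute set built from constant color, opacity, specular, emissive and roughness.
# The mask blends base into glitter, keeping the base's vertex attributes.
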MaterialFunction /AtomFramework/Materials/MatFunctions/MF_Primitive_Glit
graph {nodes [
  n1 = MaterialFunctionCall() {materialFunction=&/MF_Primitive_Trans_Base}
  FunctionOutput(None=@n1.Thin) {outputName=TransmittanceColor}
  n2 = Constant() {r=1}
  n3 = MaterialFunctionCall() {materialFunction=&/MF_GlitterLayer}
  n4 = SetMaterialAttributes!(
   MaterialAttributes
   BaseColor=@n2
   Opacity=@n2
   Specular=@n2
   EmissiveColor=Constant() {r=0.1}
   Roughness=Constant() {r=0.5}
   Normal=NamedRerouteDeclaration(None=@n3.Glitter Normal) {name=Glitter Normal, nodeColor=<rgba 0.5 0.5 1 1>}
  ) {attributeSetTypes [
    #1
    #2
    #3
    #4
    #5
    #6
   ]}
  FunctionOutput(None=BlendMaterialAttributes(
    A=SetMaterialAttributes!(MaterialAttributes=@n1.Result, Refraction=@n1.Result) {attributeSetTypes [#7]}
    B=@n4
    Alpha=NamedRerouteDeclaration(None=@n3.Glitter Mask) {name=GlitterMask, nodeColor=<rgba 1 0.2812 0 1>}
   ) {vertexAttributeBlendType=UseA}) {outputName=Material Attributes, sortPriority=10}
 ]}
