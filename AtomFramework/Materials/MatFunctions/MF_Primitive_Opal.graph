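# MF_Primitive_Opal is a material function that turns MF_Primitive_Trans_Base into a translucent opal look, outputting Material Attributes and TransmittanceColor.
# World-aligned OpalTest_N samples, sized by MPC Brick Scale, supply a fleck mask and flattenable normal, alongside a Fresnel-blended two-colour tint.
# It overrides the base's colour, roughness, normal, refraction, emissive and metallic.
# TwoSidedSign varies normal and roughness per face side, and path tracing gets extra roughness.
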MaterialFunction /AtomFramework/Materials/MatFunctions/MF_Primitive_Opal
graph {nodes [
  n1 = MaterialFunctionCall() {materialFunction=&/MF_Primitive_Trans_Base}
  n2 = Saturate(None=Add(A=NamedRerouteUsage(), B=@n1.Thin))
  FunctionOutput(None=@n2) {outputName=TransmittanceColor}
  n3 = GetMaterialAttributes(None=@n1.Result) {attributeGetTypes [
    #1
    #2
    #3
    #4
   ]}
  n4 = Saturate(None=TwoSidedSign())
  n5 = LinearInterpolate(A=Add(A=Constant() {r=0.1}, B=@n3.Roughness), B=@n3.Roughness, Alpha=@n4)
  n6 = Constant() {r=0.1}
  n7 = NamedRerouteUsage()
  n8 = Constant3Vector() {constant=<rgba 0 0 1 0>}
  n9 = Divide(A=Add(A=@n3.Normal, B=LinearInterpolate(A=@n7, B=@n8, Alpha=@n4)))
  n10 = NamedRerouteUsage()
  n11 = SetMaterialAttributes!(
   MaterialAttributes=@n1.Result
   BaseColor=Multiply(A=NamedRerouteUsage())
   Roughness=PathTracingQualitySwitch!(Normal=@n5, PathTraced=Add(A=@n6, B=@n5))
   Normal=AppendVector(A=ComponentMask(None=@n9) {g=true, r=true}, B=Constant() {r=1})
   Refraction=Add(A=@n1.Refraction, B=Multiply(A=@n10) {constB=0.05})
   EmissiveColor=Multiply(A=Multiply(A=@n2), B=@n3.EmissiveColor)
   Metallic=OneMinus(None=@n3.Metallic)
  ) {attributeSetTypes [
    #5
    #1
    #2
    #6
    #3
    #4
   ]}
  FunctionOutput(None=@n11) {outputName=Material Attributes, sortPriority=10}
  n12 = Multiply(A=CollectionParameter() {collection=/AtomFramework/Materials/Tools/MPC_BrickData, parameterName=Brick Scale}) {constB=0.5}
  n13 = WorldPosition()
  n14 = MaterialFunctionCall(Float3=MaterialFunctionCall(
    TextureObject=TextureObject() {samplerType=SAMPLERTYPE_LinearColor, texture=/AtomFramework/Materials/Textures/OpalTest_N}
    TextureSize=@n12
    WorldPosition=@n13.XYZ
   ) {materialFunction=&WorldAlignedTexture}.XYZ Texture) {materialFunction=&BreakOutFloat3Components}
  n15 = MaterialFunctionCall(Float3=MaterialFunctionCall(
    TextureObject=TextureObject() {samplerType=SAMPLERTYPE_LinearColor, texture=&OpalTest_N}
    TextureSize=Multiply(A=@n12)
    WorldPosition=@n13.XYZ
   ) {materialFunction=&WorldAlignedTexture}.XYZ Texture) {materialFunction=&BreakOutFloat3Components}
  NamedRerouteDeclaration(None=Min(A=@n14.B, B=@n15.B)) {name=Opal Fleck M, nodeColor=<rgba 0.5938 0 1 1>}
  n16 = VectorParameter() {parameterName=EdgeCol, defaultValue=<rgba 0.2814 0 0.5 1>}
  n17 = VectorParameter() {parameterName=Base Colour, defaultValue=<rgba 0 0.4249 1 1>}
  n18 = Fresnel() {baseReflectFraction=0, exponent=1}
  NamedRerouteDeclaration(None=LinearInterpolate(B=LinearInterpolate(A=@n16.RGB, B=@n17.RGB, Alpha=@n18), Alpha=NamedRerouteUsage()) {constA=0.1, constB=0.1}) {name=Opal Tint, nodeColor=<rgba 0 1 0.2734 1>}
  n19 = ConstantBiasScale(None=LinearInterpolate(A=@n14.R, B=@n15.R)) {bias=-0.5, scale=2}
  n20 = ConstantBiasScale(None=LinearInterpolate(A=@n14.G, B=@n15.G)) {bias=-0.5, scale=2}
  n21 = Constant() {r=1}
  NamedRerouteDeclaration(None=MaterialFunctionCall(
    Normal=MaterialFunctionCall(R=@n19, G=@n20, B=@n21) {materialFunction=&AppendMany}.RGB
    Flatness=ScalarParameter() {defaultValue=0.5, parameterName=Fleck Normal, sliderMax=1}
   ) {materialFunction=&FlattenNormal}.Result) {name=Opal Normal, nodeColor=<rgba 0.5 0.5 1 1>}
 ]}
