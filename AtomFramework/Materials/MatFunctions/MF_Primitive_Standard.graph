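# MF_Primitive_Standard is a material function that assembles the standard material attributes for AtomFramework primitives.
# Seeded color, surface and normal functions fill the base attributes, with tinted subsurface color and a path-tracing opacity switch.
# These pass through chained decorator, optional subsurface, tech, detail, dither, fake translucency, depth mask, bevel, underside and highlight layers.
# A named Edge Gradient derives from a bounds mask or vertex color, modulated by a warp texture.
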MaterialFunction /AtomFramework/Materials/MatFunctions/MF_Primitive_Standard
graph {nodes [
  n1 = MaterialFunctionCall!(Random Seed=NamedRerouteUsage()) {materialFunction=&/MF_PrimitiveColor}
  n2 = LinearInterpolate(B=LinearInterpolate(Alpha=NamedRerouteUsage()) {constA=0.95, constB=0.7}, Alpha=Desaturation(None=@n1.Result, Fraction=Constant() {r=1})) {constA=0.975, constB=0}
  n3 = Constant() {r=0.1}
  n4 = VectorParameter() {group=Surface, parameterName=Surface, channelNames {a=Spec, b=Metal, g=Rough Max, r=Rough Min}, defaultValue=<rgba 0.15 0.3 0 0.565>}
  n5 = MaterialFunctionCall!(
   Seed=NamedRerouteUsage()
   Min Max Roughness=AppendVector(A=@n4.Rough Min, B=@n4.Rough Min)
   Metal=@n4.Metal
   Spec=@n4.Metal
   Edge Falloff=NamedRerouteUsage()
  ) {materialFunction=&/MF_PrimitiveSurface}
  n6 = MaterialFunctionCall() {materialFunction=&/MF_PrimitiveRandom}
  n7 = SetMaterialAttributes!(
   MaterialAttributes
   BaseColor=@n1.Result
   SubsurfaceColor=MaterialFunctionCall(Base=@n1.Result, Blend=VectorParameter() {group=Surface, parameterName=Subsurface Tint, defaultValue=<rgba 0.5 0.5 0.5 0>}.RGB) {materialFunction=&Blend_Overlay}.Result
   Opacity=PathTracingQualitySwitch!(Normal=@n2, PathTraced=Add(A=@n2, B=@n3))
   Metallic=@n5.Metal
   Specular=@n5.Metal
   Roughness=@n5.Metal
   Normal=MaterialFunctionCall!(
    Flatten Edge warp
    Flatten Injection Warp N
    Random Seed=NamedRerouteDeclaration(None=@n6.Result) {name=Random Seed, nodeColor=<rgba 0.7344 1 0 1>}
   ) {materialFunction=&/MF_PrimitiveNormal}.Result
  ) {attributeSetTypes [
    #1
    #2
    #3
    #4
    #5
    #6
    #7
   ]}
  n8 = MaterialFunctionCall!(In=@n7) {materialFunction=&/MF_Decorator}
  n9 = MaterialFunctionCall!(In=StaticSwitchParameter(True=MaterialFunctionCall!(In=@n8.Texture) {materialFunction=&/MF_Subsurface}.Result, False=@n8.Texture) {defaultValue=true, group=Surface, parameterName=Apply Realistic Subsurface}) {materialFunction=&/MF_TechPrim}
  n10 = MaterialFunctionCall!(In=MaterialFunctionCall!(In=MaterialFunctionCall!(In=@n9.Result) {materialFunction=&/MF_CustomDetails}.Result) {materialFunction=&/MF_DitherTrans}.Result) {materialFunction=&/MF_FakeTranslucentInterior}
  n11 = MaterialFunctionCall!(In=MaterialFunctionCall!(In=MaterialFunctionCall!(In=@n10.Result) {materialFunction=&/MF_MaskFDepth}.Result) {materialFunction=&/MF_FakeBevels}.Result) {materialFunction=&/MF_CommonPart_Analytic}
  FunctionOutput(None=MaterialFunctionCall!(Mat In=MaterialFunctionCall!(In=@n11.Result) {materialFunction=&/MF_Undersides}.Result) {materialFunction=&/MF_Highlight}.Result) {outputName=Material Attributes}
  n12 = MaterialFunctionCall!(BoundingBoxBased_0-1_UVW, NormalWS) {materialFunction=&/MF_EdgeWarpMask}
  n13 = OneMinus(None=VertexColor().R)
  n14 = Multiply(
   A=Power(Base=StaticSwitchParameter(True=@n12.Edge Mask, False=@n13) {defaultValue=true, parameterName=Bounds Based Warp})
   B=Add(A=TextureSample() {samplerType=SAMPLERTYPE_LinearColor, texture=/AtomFramework/Materials/Textures/Warp_H}.R) {constB=0.36}
  )
  NamedRerouteDeclaration(None=Saturate(None=Power(Base=@n14) {constExponent=4})) {name=Edge Gradient, nodeColor=<rgba 0.3125 0 1 1>}
 ]}
