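# MF_Primitive_Trans is a material function that assembles a translucent primitive material from shared AtomFramework functions.
# MF_Primitive_Trans_Base supplies refraction and a Thin output, the latter feeding both inputs of MF_DualInject.
# The injected result gets Refraction set through SetMaterialAttributes, then passes through MF_Decorator to become the Material Attributes output.
# MF_DualInject's Thin output is exposed separately as TransmittanceColor.
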MaterialFunction /AtomFramework/Materials/MatFunctions/MF_Primitive_Trans
graph {nodes [
  n1 = MaterialFunctionCall() {materialFunction=&/MF_Primitive_Trans_Base}
  n2 = MaterialFunctionCall!(In_1=@n1.Thin, In=@n1.Thin) {materialFunction=&/MF_DualInject}
  FunctionOutput(None=MaterialFunctionCall!(In=SetMaterialAttributes!(MaterialAttributes=@n2.Result, Refraction=@n1.Refraction) {attributeSetTypes [#1]}) {materialFunction=&/MF_Decorator}.Texture) {outputName=Material Attributes}
  FunctionOutput(None=@n2.Thin) {outputName=TransmittanceColor}
 ]}
